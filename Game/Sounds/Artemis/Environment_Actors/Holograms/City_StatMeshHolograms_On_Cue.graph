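# Sound cue for the switch-on sound of static-mesh city holograms among the Artemis environment actors.
# A single wave player feeds a modulator, reusing the BattleLab AutoSpawner hologram-on wave.
# It plays at low volume in the gadgets and items sound class, with city hologram attenuation and a concurrency override that caps simultaneous instances.
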SoundCue /Game/Sounds/Artemis/Environment_Actors/Holograms/City_StatMeshHolograms_On_Cue
root=Modulator_0
nodes [
 {id=Modulator_0, children [WavePlayer_1]}
 {id=WavePlayer_1, properties {soundWaveAssetPtr=/Game/Sounds/BattleLab/AutoSpawner/AutoSpawner_HologramOn_01}}
]
properties {
 attenuationSettings=&/City_Hologram_Att
 bOverrideConcurrency=true
 maxDistance=1174
 soundClassObject=&Fort_Gadgets_Items
 volumeMultiplier=0.2
 attenuationOverrides {falloffDistance=512, attenuationShapeExtents {x=256}}
 concurrencyOverrides {maxCount=2}
}
